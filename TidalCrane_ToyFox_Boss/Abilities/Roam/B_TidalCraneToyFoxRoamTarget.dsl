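; Server-only roam target for the TidalCrane ToyFox boss that caches whether an area merits attack and supplies attack points.
; RefreshTargetBuildingActors, called by the roam reposition ability, async-overlaps the sphere for damageable, non-player-built buildings with centroids at or above its center.
; On completion it computes their bounding box and volume-weighted center, validating against hotfixable count and extent thresholds.
; GetTargetLocations returns random points within that box.
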(event EventBeginPlay
  ;[0 0.15 0.03878 0.5] BeginPlay: Server only actor.
  (|CallParentFunction)
  (if (not (Networking|IsDedicatedServer))
    (Actor|DestroyActor)))

(event Custom|RefreshTargetBuildingActors
  ;[0.05233 0.01747 0.15 0.5] RefreshTargetBuildingActors: Run an overlap trace to find building actors within the sphere component.
  ; Note: Sphere component itself has no collision.
  ; This is run to cache whether this roam target is valid enough for an attack from the queen.
  ; Called from GA_TidalCrane_ToyFox_Boss_Reposition_Roam
  (Class|AsyncOverlapQueue|AsyncSphereOverlapActors :Location (Transformation|GetWorldLocation (Variables|Default|GetSphere)) :Radius (Components|Sphere|GetScaledSphereRadius (Variables|Default|GetSphere)) :ShouldCancelProcess (EventDispatchers|CreateEvent :!event HandleAsyncOverlapShouldCancel) :ProcessHitActorFunction (EventDispatchers|CreateEvent :!event HandleAsyncOverlapProcessHitActor) :CompletionFunction (EventDispatchers|CreateEvent :!event HandleAsyncOverlapComplete) :TraceChannel "ECC_GameTraceChannel11"))

(fn GetTargetLocations
  (params (Count :type Integer))
  ;[0.1263 0.05236 0.15 0.5] Return random target locations in the valid extent of the roam actor target. Trying to return locations where building actors are present.
  (for _ (range Count)
    (Utilities|Array|Add (LocalVariables|GetRetLocations) (Math|Random|RandomPointinBoundingBox (Variables|Transient|GetTargetBoxExtentCenter) (Variables|Transient|GetTargetBoxExtent))))
  (return (LocalVariables|GetRetLocations)))

(fn HandleAsyncOverlapComplete
  (params (WasCancelled :type Boolean))
  (if (not WasCancelled)
    (CallFunction|ProcessOverlapActors)))

(fn HandleAsyncOverlapProcessHitActor
  (params (HitActor :type Actor%20Object%20Reference))
  (bind (_returnvalue_x _returnvalue_y _returnvalue_z) (Math|Vector|BreakVector (Transformation|GetWorldLocation (Variables|Default|GetSphere))))
  ;[0.01697 0.02577 0.15 0.5] HandleAsyncOverlapProcessHitActor: Only look for valid building actors that are not player built and can be damaged.
  (bind _asbuilding_actor (Utilities|Casting|CastToBuildingActor HitActor))
  (if (Math|Boolean|ANDBoolean (Variables|/Script/Engine.Actor|GetbCanBeDamaged _asbuilding_actor) (not (BuildingActor|IsPlayerBuilt _asbuilding_actor)) (>= (.z (Building|GetCentroid _asbuilding_actor)) _returnvalue_z))
    (Utilities|Array|Add (Variables|Transient|GetTargetBuildingActors) _asbuilding_actor)))

(fn HandleAsyncOverlapShouldCancel
  (return (Game|IsActorBeingDestroyed)))

(fn IsValidRoamTarget
  ;[0.05049 0.08109 0.15 0.5] Returns if the roam target is valid for attack. It is valid for an attack if there's enough building actors in here.
  (return (Variables|Transient|GetbValidRoamTarget)))

(fn ProcessOverlapActors
  (bind (_minvector_x _minvector_y _minvector_z) (Math|Vector|BreakVector (Variables|GetMinVector_X)))
  (bind (_maxvector_x _maxvector_y _maxvector_z) (Math|Vector|BreakVector (Variables|GetMaxVector_X)))
  (bind _maxvector (LocalVariables|GetMaxVector))
  ;[0.136 0.04535 0.15 0.5] ProcessOverlapActors: Loop through the overlapped building actors and check if there's enough building actors to attack.
  (for _array_element (Variables|Transient|GetTargetBuildingActors)
    (Utilities|IsValid _array_element
      (:Is%20Valid
        (bind (_origin _boxextent) (Collision|GetActorBounds _array_element false false))
        (bind _output_get (LocalVariables|SetActorMin (- _origin _boxextent)))
        (bind _output_get_2 (LocalVariables|SetActorMax (+ _origin _boxextent)))
        (Variables|SetMinVector_X (Math|Vector|MakeVector (Math|Float|Min(Float) _minvector_x (.x _output_get)) (Math|Float|Min(Float) _minvector_y (.y _output_get)) (Math|Float|Min(Float) _minvector_z (.z _output_get))))
        (Variables|SetMaxVector_X (Math|Vector|MakeVector (Math|Float|Max(Float) _maxvector_x (.x _output_get_2)) (Math|Float|Max(Float) _maxvector_y (.y _output_get_2)) (Math|Float|Max(Float) _maxvector_z (.z _output_get_2))))
        (bind _output_get_3 (LocalVariables|SetVolume (Utilities|Operators|Multiply (* (.x _boxextent) 2.0) (* (.y _boxextent) 2.0) (* (.z _boxextent) 2.0))))
        (LocalVariables|SetWeightedLocationSum (+ (LocalVariables|GetWeightedLocationSum) (* (Building|GetCentroid _array_element) _output_get_3)))
        (LocalVariables|SetTotalWeight (+ _output_get_3 (LocalVariables|GetTotalWeight))))))
  (bind _output_get_4 (Variables|Transient|SetTargetCenterOfMass (/ (LocalVariables|GetWeightedLocationSum) (LocalVariables|GetTotalWeight))))
  (bind _output_get_5 (Variables|Transient|SetTargetBoxExtentCenter (/ (+ (LocalVariables|GetMinVector) _maxvector) 2.0)))
  (bind _output_get_6 (Variables|Transient|SetTargetBoxExtent (/ (- _maxvector (LocalVariables|GetMinVector)) 2.0)))
  (Rendering|Debug|DrawDebugSphere _output_get_4 1e+02 12 (select (Networking|IsDedicatedServer) "(R=0.000000,G=1.000000,B=0.021790,A=1.000000)" "(R=0.000000,G=1.000000,B=0.933321,A=1.000000)") 8.0 5.0)
  (Rendering|Debug|DrawDebugBox :Center _output_get_5 :Extent _output_get_6 :Duration 8.0 :Thickness 5.0 :DepthPriority "World")
  (Variables|Transient|SetbValidRoamTarget (and (> (Utilities|Array|Length (Variables|Transient|GetTargetBuildingActors)) (Math|Float|Truncate (ScalableFloat|GetCachedScalableFloatValue (Variables|Hotfix|GetHotfix_ValidThresholdNumOfBuildingActors)))) (> (Math|Vector|VectorLengthSquared (Variables|Transient|GetTargetBoxExtent)) (Math|Float|Square (ScalableFloat|GetCachedScalableFloatValue (Variables|Hotfix|GetHotfix_ValidThresholdBoxExtentVolume)))))))

(fn ConstructionScript)
