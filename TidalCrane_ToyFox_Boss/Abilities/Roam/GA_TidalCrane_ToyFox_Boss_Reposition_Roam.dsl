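; Roam reposition ability for the TidalCrane ToyFox boss, extending a parent ability to burrow it and resurface it near a new attack target.
; After burrowing (victory montage the first time), it checks a hotfixable attack cap, waits randomly, then checks safe-zone phase.
; Chained EQS queries pick a registered roam target, then a nearby reposition actor to move to.
; Reached limits or failed queries silently kill the boss and end the ability.
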(event EventBeginRepositionQuery
  ;[0 0.15 0.04559 0.5] BeginRepositionQuery: Burrow the queen, wait a certain duration, after which find a roam target to attack using EQS.
  (CallFunction|StartBurrowSequence true) ; Put the boss into the ground.
)

(event EventNotifyBegin
  (params (NotifyTag :type Gameplay%20Tag%20Structure))
  (bind _returnvalue (SafeZone|GetSafeZoneIndicator))
  ;[0 0.15 0.04559 0.5] BeginRepositionQuery: Burrow the queen, wait a certain duration, after which find a roam target to attack using EQS.
  (|CallParentFunction NotifyTag)
  (if (== NotifyTag "(TagName=\"Event.TidalCrane.Boss.BurrowingComplete\")")
    (bind _result (Math|Integer|IncrementInt (Variables|Transient|GetRoamAttackCount))
      (:%20%20
        (if (<= _result (Math|Float|Truncate (ScalableFloat|GetCachedScalableFloatValue (Variables|Hotfix|GetHotfix_RoamAttackMaxCount))))
          (Utilities|FlowControl|Delay (Math|Random|RandomFloatinRange (ScalableFloat|GetCachedScalableFloatValue (Variables|Hotfix|GetHotfix_RoamAttackDelayMinimum)) (ScalableFloat|GetCachedScalableFloatValue (Variables|Hotfix|GetHotfix_RoamAttackDelayMaximum))))
          (Utilities|IsValid _returnvalue
            (:Is%20Valid
              (if (not (< (SafeZone|GetCurrentPhaseIndex _returnvalue) (Math|Float|Truncate (ScalableFloat|GetCachedScalableFloatValue (Variables|Hotfix|GetHotfix_RoamAttackMaxSafeZonePhaseIndex)))))
                ;;[0.01462 0.1072 0.15 0.5] SafeZonePhase index limit reached.
                (Fortnite|FortShippingLog self "BeginRepositionQuery: Max SafeZone Phase Hit.")
                (CallFunction|DestroyBossActorandEndAbility)))
            (:Is%20Not%20Valid
              _))
          (bind _outactors (Class|FortGlobalActorRegistrySubsystem|GetRegisteredActors (WorldSubsystems|GetFortGlobalActorRegistrySubsystem) (GameplayTags|MakeLiteralGameplayTag "(TagName=\"TidalCrane.ToyFox.Boss.Ability.Roam.Target\")")))
          (Utilities|RefreshRoamTargetActors _outactors
            (:Complete
              ;;[0.07582 0.03907 0.15 0.5] Find a valid roam target actor to attack.
              (bind _returnvalue_2 (AI|EQS|RunEQSQuery (Variables|Settings|GetRoamTargetEnvQuery) (Ability|GetAvatarActorfromActorInfo)))
              (AI|EQS|BindEventtoOnQueryFinished _returnvalue_2 (EventDispatchers|CreateEvent :!event Custom|HandleTargetActorFindQueryComplete)))) ; Refresh the roam target actors so they \nhave the latest information for EQS.
          (else
            ;;[0.03685 0.07732 0.15 0.5] Roam attacked the desired number of times - destroy actor. The boss should be underground at this point and thus this action should be safe.
            (Fortnite|FortShippingLog self (Utilities|String|Append "BeginRepositionQuery: RoamAttackCount Max limit hit. " (Utilities|String|ToString(Integer) (- (Variables|Transient|GetRoamAttackCount) 1))))
            (CallFunction|DestroyBossActorandEndAbility)))))))

(event Custom|HandleTargetActorFindQueryComplete
  (params (QueryInstance :type EQS%20Query%20Instance%20Object%20Reference) (QueryStatus :type EEnvQueryStatus%20Enum :default "Processing"))
  ;[0.05581 0.05882 0.15 0.5] HandleTargetActorFindQueryComplete:
  (Utilities|IsValid QueryInstance
    (:Is%20Valid
      (bind _resultactors (AI|EQS|GetQueryResultsasActors QueryInstance))
      (if (Utilities|Array|IsValidIndex _resultactors)
        (Utilities|IsValid (Utilities|Array|Get(acopy) _resultactors)
          (:Is%20Valid
            (bind _bossactor (Variables|InternalData|GetBossActor)
              (:then
                (Class|BPITidalCraneBoss|SetCurrentTargetActor(Message) _bossactor (Utilities|Array|Get(acopy) _resultactors))
                ;;[0.05767 0.15 0.05139 0.5] Find a valid reposition actor around the selected roam target actor to move to.
                (bind _returnvalue (AI|EQS|RunEQSQuery (Variables|Settings|GetRoamRepositionEnvQuery) (Utilities|Array|Get(acopy) _resultactors) "RandomBest5Pct"))
                (AI|EQS|BindEventtoOnQueryFinished _returnvalue (EventDispatchers|CreateEvent :!event RepositioningEQSFinished)))
              (:else)))
          (:Is%20Not%20Valid
            _))
        (else
          _))
      (Fortnite|FortShippingLog self "HandleTargetActorFindQueryComplete: Query failed, destroying self.")
      (CallFunction|DestroyBossActorandEndAbility))
    (:Is%20Not%20Valid
      _)))

(event EventRepositioningEQSFinished
  (params (QueryInstance :type EQS%20Query%20Instance%20Object%20Reference) (QueryStatus :type EEnvQueryStatus%20Enum :default "Processing"))
  ;[0.04308 0.15 0.05045 0.5] RepositioningEQSFinished: Rise out of the selected reposition actor.
  (Utilities|IsValid QueryInstance
    (:Is%20Valid
      (bind _resultactors (AI|EQS|GetQueryResultsasActors QueryInstance))
      (if (Utilities|Array|IsValidIndex _resultactors)
        (bind _asb_tidal_crane_toy_fox_reposition_actor (Utilities|Casting|CastToB_TidalCraneToyFoxRepositionActor (Utilities|Array|Get(acopy) _resultactors))
          (:then
            (bind _output_get (Variables|Transient|SetSelectedRepositionActor _asb_tidal_crane_toy_fox_reposition_actor))
            (Variables|Default|SetNewLocation (Transformation|GetActorLocation _output_get))
            (CallFunction|MovetoNewLocation)
            (Ability|CommitAbility))
          (:CastFailed
            _))
        (else
          _))
      (Fortnite|FortShippingLog self "RepositioningEQSFinished: Query failed, destroying self.")
      (CallFunction|DestroyBossActorandEndAbility))
    (:Is%20Not%20Valid
      _)))

(event EventStartBurrowSequence
  (params (bAutoMoveToNewLocation :type Boolean))
  ;[0.15 0.07752 0.05795 0.5] StartBurrowSequence: The first time this is called, it means that the giant was eliminated, in that case, we should play the victory version of the burrow montage.
  (Variables|Animation|SetMontageInfo (select (Variables|Transient|GetbFirstExecution) (Variables|Settings|GetVictoryBurrowMontageInfo) (Variables|Settings|GetDefaultBurrowMontageInfo)))
  (|CallParentFunction bAutoMoveToNewLocation)
  (Variables|Transient|SetbFirstExecution false))

(fn DestroyBossActorAndEndAbility
  ;[0.04075 0.1342 0.15 0.5] Destroy the boss actor and end the ability.
  (Fortnite|FortShippingLog self "DestroyBaseActorAndEndAbility: Invoked")
  (bind _bossactor (Variables|InternalData|GetBossActor)
    (:then
      (bind _asbuilding_actor (Utilities|Casting|CastToBuildingActor _bossactor))
      (Gameplay|Die _asbuilding_actor "(TagName=\"TidalCrane.ToyFox.Finisher\")" 0 0 :!class /Script/FortniteGame.BuildingActor) ; Specific case as ATidalCraneSwarmBoss class :\nWe will simulate a SilentDie() in code when the tag "TidalCrane.ToyFox.Finisher" is used.
      (Ability|EndAbility))
    (:else
      (Fortnite|FortShippingLog self "DestroyBaseActorAndEndAbility: BossActor is null." true))))

(fn CanActivateAbility
  (params (ActorInfo :type Gameplay%20Ability%20Actor%20Info%20Structure) (Handle :type Gameplay%20Ability%20Spec%20Handle%20Structure))
  (return 0 true))
